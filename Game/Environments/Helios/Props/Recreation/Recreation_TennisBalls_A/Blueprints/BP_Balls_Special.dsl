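; BP_Balls_Special is a tennis ball prop Blueprint that randomly decides at runtime whether it stays in the level.
; On BeginPlay it rolls a random integer between the RollMin and RollMax variables and destroys itself unless the result is 1.
; Its construction script only calls the parent implementation.
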(event EventBeginPlay
  (if (not (== (Math|Random|RandomIntegerinRange (Variables|Default|GetRollMin) (Variables|Default|GetRollMax)) 1))
    (Actor|DestroyActor)))

(fn ConstructionScript
  (|CallParentFunction))
